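; Trajectory preview for the Deployed Cover grenade, showing where its three cover pieces (left wing, center, right wing) would spawn.
; On setup it gets the weapon from the owning ability, attaches the pieces to the target and offsets them by the weapon's chunk transforms.
; Visibility updates move the target to the spawn transform, swap spline materials and color each piece by its validity.
; An invalid trajectory hides all pieces and shows the invalid-target marker.
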(event EventBeginPlay
  (|CallParentFunction)
  (Variables|Default|SetPreview%20Meshes (Utilities|Array|MakeArray (Variables|Default|GetLeftWing) (Variables|Default|GetRightWing) (Variables|Default|GetCenterPiece))))

(event EventUpdateTargetMeshVisibility
  (bind _targetvalidity (Variables|Default|GetTargetValidity))
  (bind _bistrajectoryvalid (Variables|Default|GetbIsTrajectoryValid))
  (Rendering|SetVisibility (Variables|Default|GetInvalidTarget) (not (Variables|Default|GetbIsTrajectoryValid)) false)
  (Utilities|FlowControl|Sequence
    (:then_0
      ; Turn preview to spawn location and rotation (camera yaw + surface pitch/roll)
      (Transformation|SetWorldTransform (Variables|Default|GetTarget) (Math|Transform|MakeTransform (Variables|Default|GetTargetLocation) (Variables|Default|GetTargetRotation)) false false))
    (:then_1
      (for _array_element (Variables|Trajectory|GetSplineMeshComponents)
        (Rendering|Material|SetMaterial _array_element 0 (Selection|SelectMaterialInstance (Variables|VFX|GetSplineMID) (Variables|VFX|GetInvalidSplineMID) (Variables|Default|GetbIsTrajectoryValid)))))
    (:then_2
      ; Update preview meshes to latest validity and pitch angle. Set the color of the preview mesh based on if it is a valid location or not. If center piece is not at a valid location just show the invalid target and hide all preview meshes.
      (for (_array_element_2 _array_index) (Variables|Default|GetPreview%20Meshes)
        (if (and _bistrajectoryvalid (Utilities|Array|IsValidIndex _targetvalidity _array_index))
          (Rendering|Material|SetColorParameterValueonMaterials _array_element_2 "Color" (Math|Color|SelectColor (Variables|Default|GetValidTrajectoryColor) (Variables|Default|GetInvalidTrajectoryColor) (Math|Boolean|ANDBoolean _bistrajectoryvalid (Utilities|Array|IsValidIndex _targetvalidity _array_index) (Utilities|Array|Get(aref) _targetvalidity _array_index))))
          (if _bistrajectoryvalid
            (Rendering|SetVisibility _array_element_2 _bistrajectoryvalid false))
          (else
            (Rendering|SetVisibility _array_element_2 _bistrajectoryvalid false)))))))

(fn SetShouldUpdateFromOwner
  (params (ShouldUpdate :type Boolean) (Owner :type Object%20Reference))
  (bind (_player _weapon _bpreviewing _previewdistance _maxdistance _mindistance _mindistfromindestructible _benableblockedbyindestructibles _playerforwardvectorscale _initialtracepullbackdist _downwardtracedistancenear _downwardtracedistancenearup _downwardtracedistancefar _bstopped _classestoignorewhentracing _classestoignorewhentracingiftoosmall _debugdrawtime _chunkrelativetransforms) (Utilities|Struct|BreakItemizedPropSpawnerWeaponData (Variables|/DeployedCoverGameplay/Items/ItemSpawner/B_DeployedCoverGrenade.B_DeployedCoverGrenade_C|GetNativeData (Variables|Default|GetDeployedCoverWeapon))))
  (|CallParentFunction ShouldUpdate Owner)
  (bind _asfort_gameplay_ability (Utilities|Casting|CastToFortGameplayAbility Owner))
  (bind _asb_deployed_cover_grenade (Utilities|Casting|CastToB_DeployedCoverGrenade (Ability|GetCurrentSourceWeapon _asfort_gameplay_ability)))
  (Variables|Default|SetDeployedCoverWeapon _asb_deployed_cover_grenade)
  (for (_array_element _array_index) (Variables|Default|GetPreview%20Meshes)
    (Transformation|AttachComponentToComponent _array_element (Variables|Default|GetTarget) "None" "KeepRelative" "KeepRelative" "KeepRelative" true)
    (Transformation|SetRelativeTransform _array_element (Utilities|Array|Get(aref) _chunkrelativetransforms _array_index) false true)))

(fn ConstructionScript
  (|CallParentFunction))
